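# Looping flight sound cue for the Athena airship, built as a distant layer and a near layer, both in the Fort_Gadgets sound class.
# The distant layer is a pitched-down Battle Bus propeller wind loop with very long falloff, low-pass filtering and listener focus.
# The near layer mixes bus wind, a pitched-down jetpack loop and a pitched-up airship blade loop under much shorter falloff.
# Separate attenuation overrides per branch let the sound's makeup change with distance.
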SoundCue /Figment_S09/Athena/Sounds/Vehicles/Airship/Athena_Airship_Flying_Loop_Cue
root=Mixer_0
nodes [
 {
  id=Attenuation_0
  children [Modulator_0]
  properties {
   bOverrideAttenuation=true
   attenuationOverrides {
    bAttenuateWithLPF=true
    bEnableFocusInterpolation=true
    bEnableListenerFocus=true
    falloffDistance=300000
    focusAttackInterpSpeed=0.2
    focusAzimuth=2
    focusReleaseInterpSpeed=0.2
    lPFFrequencyAtMax=6000
    lPFFrequencyAtMin=6000
    nonFocusAzimuth=6
    nonFocusVolumeAttenuation=0
    occlusionTraceChannel=ECC_GameTraceChannel11
    reverbWetLevelMax=10
    reverbWetLevelMin=10
    stereoSpread=5000
    attenuationShapeExtents {x=0}
   }
  }
 }
 {
  id=Attenuation_1
  children [Mixer_2]
  properties {
   bOverrideAttenuation=true
   attenuationOverrides {
    bEnableFocusInterpolation=true
    falloffDistance=15000
    focusAzimuth=5
    nonFocusAzimuth=10
    nonFocusVolumeAttenuation=0.2
    reverbWetLevelMax=10
    reverbWetLevelMin=10
    stereoSpread=2048
    attenuationShapeExtents {x=0}
   }
  }
 }
 {id=Mixer_0, children [SoundClass_1,SoundClass_0], properties {inputVolume [0.2,0.5]}}
 {id=Mixer_2, children [WavePlayer_7,Modulator_3,Modulator_1], properties {inputVolume [1.5,0.3,0.4]}}
 {id=Modulator_0, children [WavePlayer_11], properties {pitchMax=0.65, pitchMin=0.65}}
 {id=Modulator_1, children [WavePlayer_0], properties {pitchMax=1.2, pitchMin=1.2, volumeMax=1, volumeMin=1}}
 {id=Modulator_3, children [WavePlayer_12], properties {pitchMax=0.8, pitchMin=0.8, volumeMax=1, volumeMin=1}}
 {id=SoundClass_0, children [Attenuation_1], properties {soundClassOverride=&Fort_Gadgets}}
 {id=SoundClass_1, children [Attenuation_0], properties {soundClassOverride=&Fort_Gadgets}}
 {id=WavePlayer_0, properties {bLooping=true, soundWaveAssetPtr=&/Airship_Blade_RotateLoop_01}}
 {id=WavePlayer_11, properties {bLooping=true, soundWaveAssetPtr=/Game/Athena/Sounds/Vehicles/Bus/SFX_FlyingBus_Propeller_Wind_Loop}}
 {id=WavePlayer_12, properties {bLooping=true, soundWaveAssetPtr=/Game/Athena/Sounds/Items/Jetpack/Jetpack_FlyLoop_01}}
 {id=WavePlayer_7, properties {bLooping=true, soundWaveAssetPtr=/Game/Athena/Sounds/Vehicles/Bus/SFX_Bus_Flying_Wind_Loop}}
]
properties {maxDistance=300000, volumeMultiplier=1}
